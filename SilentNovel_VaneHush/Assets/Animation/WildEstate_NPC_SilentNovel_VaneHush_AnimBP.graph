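# Item-override animation Blueprint for the VaneHush NPC on the Fortnite player skeleton, adding alert animations for when it stands still.
# Only ItemFullBodyOverride has logic: a state machine plays an unready-to-alert intro, then an alert pose sequence, while Unaware, Combat and Dead reuse the cached input pose.
# The update event gates alert entry on bAlert and not moving; movement or lost alert returns to Unaware.
# bInCombat drives Combat, and a global alias enters Dead.
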AnimBlueprint /SilentNovel_VaneHush/Assets/Animation/WildEstate_NPC_SilentNovel_VaneHush_AnimBP
target_skeleton=&Fortnite_M_Avg_Player_Skeleton
graphs [
 AnimGraph {nodes [Output Pose()]}
 EventGraph {nodes [
   Event BlueprintUpdateAnimation()
   n1 = Get bIsMoving2D()
   > Set CanEnterAlertedState(CanEnterAlertedState:bool=AND Boolean(A=Get bAlert() -> (bAlert:bool), B=NOT Boolean(A=@n1))) -> (Output_Get:bool)
  ]}
 ItemAimOffset {nodes [
   Input Pose!() -> (Pose, AimOffsetAlpha:double, Yaw:double, Pitch:double) {name=InPose_AimOffset}
   Input Pose!() -> (Pose) {name=InPose_PreAimOffsets}
   Output Pose() {layerGroup=Item Override}
  ]}
 ItemFallAdditive {nodes [
   Input Pose() {name=InPoseItemFallAdditive}
   Output Pose() {layerGroup=Item Override}
  ]}
 ItemFallLandAdditive {nodes [
   Input Pose() {name=InPoseItemFallLandAdditive}
   Output Pose() {layerGroup=Item Override}
  ]}
 ItemFallLevel2Override {nodes [
   Input Pose() {name=InPose_FallLevel2}
   Output Pose() {layerGroup=Item Override}
  ]}
 ItemFinalPoseOverride {nodes [
   Input Pose() {name=InFinalPose}
   Output Pose() {layerGroup=Item Override}
  ]}
 ItemFullBodyAimOffsets {nodes [
   Input Pose!() -> (Pose, FullBody_AimOffsetAlpha:double, Aim_Yaw:double, Aim_Pitch:double) {name=InPoseFullBodyAimOffsets}
   Input Pose!() -> (Pose) {name=InPosePreFullBodyAimOffsets}
   Output Pose() {layerGroup=Item Override}
  ]}
 ItemFullBodyOverride {nodes [
   AnimGraphNode_StateMachine_0 = StateMachine() {is=title AnimGraphNode_StateMachine}
   Output Pose(Result=@AnimGraphNode_StateMachine_0) {layerGroup=Item Override}
   Save cached pose 'FullBodyCache'(Pose=Input Pose() {name=InPoseFullBody}) {cacheName=FullBodyCache}
  ]}
 StateMachine {
  entry=UnawareLoop
  path=ItemFullBodyOverride.AnimGraphNode_StateMachine_0.StateMachine
  ids=id
  nodes [
   ToDeath() {is=title AnimStateAliasNode, bGlobalAlias=true}
   AnimStateNode_0 = UnawareLoop() {is=title AnimStateNode}
   AnimStateNode_1 = ToAlerted() {is=title AnimStateNode}
   AnimStateNode_2 = Alerted() {is=title AnimStateNode}
   AnimStateNode_3 = Combat() {is=title AnimStateNode}
   AnimStateNode_4 = Dead() {is=title AnimStateNode}
   AnimStateTransitionNode_0 = AnimStateTransitionNode() {bAllowInertializationForSelfTransitions=true}
   AnimStateTransitionNode_1 = AnimStateTransitionNode() {bAllowInertializationForSelfTransitions=true, bAutomaticRuleBasedOnSequencePlayerInState=true}
   AnimStateTransitionNode_2 = AnimStateTransitionNode() {bAllowInertializationForSelfTransitions=true}
   AnimStateTransitionNode_24 = AnimStateTransitionNode() {bAllowInertializationForSelfTransitions=true, priorityOrder=4}
   AnimStateTransitionNode_25 = AnimStateTransitionNode() {bAllowInertializationForSelfTransitions=true}
   AnimStateTransitionNode_3 = AnimStateTransitionNode() {bAllowInertializationForSelfTransitions=true}
   AnimStateTransitionNode_4 = AnimStateTransitionNode() {bAllowInertializationForSelfTransitions=true}
   AnimStateTransitionNode_5 = AnimStateTransitionNode() {bAllowInertializationForSelfTransitions=true}
   AnimStateTransitionNode_7 = AnimStateTransitionNode() {bAllowInertializationForSelfTransitions=true}
   AnimStateTransitionNode_8 = AnimStateTransitionNode() {bAllowInertializationForSelfTransitions=true}
  ]
  links [
   {
    dst=Combat
    rule=ItemFullBodyOverride.AnimGraphNode_StateMachine_0.StateMachine.AnimStateTransitionNode_2.Transition
    src=Alerted
    via=AnimStateTransitionNode_2
   }
   {
    dst=UnawareLoop
    rule=ItemFullBodyOverride.AnimGraphNode_StateMachine_0.StateMachine.AnimStateTransitionNode_3.Transition
    src=Alerted
    via=AnimStateTransitionNode_3
   }
   {
    dst=UnawareLoop
    rule=ItemFullBodyOverride.AnimGraphNode_StateMachine_0.StateMachine.AnimStateTransitionNode_7.Transition
    src=Alerted
    via=AnimStateTransitionNode_7
   }
   {
    dst=UnawareLoop
    rule=ItemFullBodyOverride.AnimGraphNode_StateMachine_0.StateMachine.AnimStateTransitionNode_25.Transition
    src=Combat
    via=AnimStateTransitionNode_25
   }
   {
    dst=Alerted
    rule=ItemFullBodyOverride.AnimGraphNode_StateMachine_0.StateMachine.AnimStateTransitionNode_1.Transition
    src=ToAlerted
    via=AnimStateTransitionNode_1
   }
   {
    dst=UnawareLoop
    rule=ItemFullBodyOverride.AnimGraphNode_StateMachine_0.StateMachine.AnimStateTransitionNode_3.Transition
    src=ToAlerted
    via=AnimStateTransitionNode_4
   }
   {
    dst=UnawareLoop
    rule=ItemFullBodyOverride.AnimGraphNode_StateMachine_0.StateMachine.AnimStateTransitionNode_7.Transition
    src=ToAlerted
    via=AnimStateTransitionNode_8
   }
   {
    dst=Dead
    rule=ItemFullBodyOverride.AnimGraphNode_StateMachine_0.StateMachine.AnimStateTransitionNode_5.Transition
    src=ToDeath
    via=AnimStateTransitionNode_5
   }
   {
    dst=Combat
    rule=ItemFullBodyOverride.AnimGraphNode_StateMachine_0.StateMachine.AnimStateTransitionNode_24.Transition
    src=UnawareLoop
    via=AnimStateTransitionNode_24
   }
   {
    dst=ToAlerted
    rule=ItemFullBodyOverride.AnimGraphNode_StateMachine_0.StateMachine.AnimStateTransitionNode_0.Transition
    src=UnawareLoop
    via=AnimStateTransitionNode_0
   }
  ]
 }
 UnawareLoop {
  path=ItemFullBodyOverride.AnimGraphNode_StateMachine_0.StateMachine.AnimStateNode_0.UnawareLoop
  nodes [Output Animation Pose(Result=Use cached pose 'FullBodyCache'())]
 }
 ToAlerted {
  path=ItemFullBodyOverride.AnimGraphNode_StateMachine_0.StateMachine.AnimStateNode_1.ToAlerted
  nodes [Output Animation Pose(Result=Sequence Player 'SilentNovel_VaneHush_UnreadyToAlert'() {bLoopAnimation=false, sequence=/SilentNovel_VaneHush/Assets/Animation/Actions/SilentNovel_VaneHush_UnreadyToAlert})]
 }
 Alerted {
  path=ItemFullBodyOverride.AnimGraphNode_StateMachine_0.StateMachine.AnimStateNode_2.Alerted
  nodes [Output Animation Pose(Result=Sequence Player 'SilentNovel_VaneHush_Unready_Alert'() {bLoopAnimation=false, sequence=/SilentNovel_VaneHush/Assets/Animation/Actions/SilentNovel_VaneHush_Unready_Alert})]
 }
 Combat {
  path=ItemFullBodyOverride.AnimGraphNode_StateMachine_0.StateMachine.AnimStateNode_3.Combat
  nodes [Output Animation Pose(Result=Use cached pose 'FullBodyCache'())]
 }
 Dead {
  path=ItemFullBodyOverride.AnimGraphNode_StateMachine_0.StateMachine.AnimStateNode_4.Dead
  nodes [Output Animation Pose(Result=Use cached pose 'FullBodyCache'())]
 }
 Transition {
  path=ItemFullBodyOverride.AnimGraphNode_StateMachine_0.StateMachine.AnimStateTransitionNode_0.Transition
  nodes [Result(bCanEnterTransition=Get CanEnterAlertedState() -> (CanEnterAlertedState:bool))]
 }
 Transition {path=ItemFullBodyOverride.AnimGraphNode_StateMachine_0.StateMachine.AnimStateTransitionNode_1.Transition, nodes [Result()]}
 Transition {
  path=ItemFullBodyOverride.AnimGraphNode_StateMachine_0.StateMachine.AnimStateTransitionNode_2.Transition
  nodes [Result(bCanEnterTransition=Get bInCombat() -> (bInCombat:bool))]
 }
 Transition {
  path=ItemFullBodyOverride.AnimGraphNode_StateMachine_0.StateMachine.AnimStateTransitionNode_24.Transition
  nodes [Result(bCanEnterTransition=Get bInCombat())]
 }
 Transition {
  path=ItemFullBodyOverride.AnimGraphNode_StateMachine_0.StateMachine.AnimStateTransitionNode_25.Transition
  nodes [Result(bCanEnterTransition=NOT Boolean(A=Get bInCombat()))]
 }
 Transition {
  path=ItemFullBodyOverride.AnimGraphNode_StateMachine_0.StateMachine.AnimStateTransitionNode_3.Transition
  nodes [Result(bCanEnterTransition=NOT Boolean(A=Get bAlert()))]
 }
 Transition {
  path=ItemFullBodyOverride.AnimGraphNode_StateMachine_0.StateMachine.AnimStateTransitionNode_5.Transition
  nodes [Result(bCanEnterTransition=Get bDead() -> (bDead:bool))]
 }
 Transition {
  path=ItemFullBodyOverride.AnimGraphNode_StateMachine_0.StateMachine.AnimStateTransitionNode_7.Transition
  nodes [Result(bCanEnterTransition=Get bIsMoving2D())]
 }
 ItemIdleOverride {nodes [
   Input Pose() {name=InPose_Idle}
   Output Pose() {layerGroup=Item Override}
  ]}
 ItemJetPackJumpAdditive {nodes [Output Pose() {layerGroup=Item Override}]}
 ItemJetPackStartAdditive {nodes [Output Pose() {layerGroup=Item Override}]}
 ItemJumpLoopAdditive {nodes [
   Input Pose() {name=InPoseItemJumpLoopAdditive}
   Output Pose() {layerGroup=Item Override}
  ]}
 ItemJumpOverride {nodes [
   Input Pose() {name=InPoseJump}
   Output Pose() {layerGroup=Item Override}
  ]}
 ItemJumpUpAdditive {nodes [
   Input Pose() {name=InPoseItemJumpUpAdditive}
   Output Pose() {layerGroup=Item Override}
  ]}
 ItemPreAimOffsets {nodes [
   Input Pose() {name=InPosePreAimOffsets}
   Output Pose() {layerGroup=Item Override}
  ]}
 ItemPreIK {nodes [
   Input Pose() {name=InPose_PreIK}
   Output Pose() {layerGroup=Item Override}
  ]}
 ItemRightHandIK {nodes [
   Input Pose() {name=InPosePreRightHandIK}
   Input Pose() {name=InPosePostRightHandIK}
   Output Pose() {layerGroup=Item Override}
  ]}
 ItemSkeletalControl {nodes [
   Input Pose()
   Output Pose() {layerGroup=Item Override}
  ]}
 ItemSwimJumpFallLoop {nodes [Output Pose() {layerGroup=Item Override}]}
 ItemSwimJumpSurfaceEnd {nodes [Output Pose() {layerGroup=Item Override}]}
 ItemSwimJumpSurfaceLoop {nodes [Output Pose() {layerGroup=Item Override}]}
 ItemUpperBody {nodes [
   Input Pose!() -> (Pose, InputParam:Fort Anim Input Adjusted Aim Structure) {name=InPoseUpperBody}
   Output Pose() {layerGroup=Item Override}
  ]}
 ItemZipLineStartAdditive {nodes [Output Pose() {layerGroup=Item Override}]}
]
variables [CanEnterAlertedState {type=boolean}]
